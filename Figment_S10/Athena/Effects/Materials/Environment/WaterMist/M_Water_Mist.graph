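# Two-sided translucent water mist material with volumetric directional lighting and a time-of-day tint scaled by an emissive multiplier.
# Three panning world-space noise layers share a swirl-noise distortion, are powered and averaged, then masked by contrast-adjusted vertex alpha.
# Inverse Fresnel and camera depth fade soften that opacity, which feeds a DepthFade whose distance varies with another panning noise.
# Vertex alpha also scales the Z of a vertex-normal-based vector.
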Material /Figment_S10/Athena/Effects/Materials/Environment/WaterMist/M_Water_Mist
graph {
 properties {
  blendMode=BLEND_Translucent
  pixelDepthOffsetMode=PDOM_Legacy
  translucencyLightingMode=TLM_VolumetricDirectional
  twoSided=true
  usedShadingModels=Default Lit
 }
 nodes [
  n1 = VertexNormalWS()
  n2 = VertexColor()
  n3 = ScalarParameter() {parameterName=Z Depth Multiplier}
  AppendVector(A=ComponentMask(None=@n1) {g=true, r=true}, B=Multiply(A=ComponentMask(None=@n1) {b=true}, B=Multiply(A=@n2.A, B=@n3) {constB=25}))
  n4 = OneMinus(None=Fresnel(ExponentIn=ScalarParameter() {defaultValue=3, parameterName=Fresnel Exponent}))
  n5 = ScalarParameter() {parameterName=2352tergd}
  n6 = Multiply(A=WorldPosition().XYZ, B=ScalarParameter() {defaultValue=0.0003, parameterName=World UV Scale})
  n7 = Multiply(A=Time(), B=AppendVector(A=ScalarParameter() {parameterName=Tex 1 Speed U}, B=ScalarParameter() {parameterName=Tex 1 Speed V}))
  n8 = ComponentMask(None=Multiply(A=WorldPosition().XYZ, B=ScalarParameter() {defaultValue=0.0003, parameterName=World Distort UV Scale})) {g=true, r=true}
  n9 = ScalarParameter() {defaultValue=0.0001, parameterName=Tex Distort Speed U}
  n10 = ScalarParameter() {defaultValue=0.0001, parameterName=Tex Distort Speed V}
  n11 = Frac(None=Multiply(A=Time(), B=AppendVector(A=@n9, B=@n10)))
  n12 = Multiply(
   A=TextureSample(UVs=Add(A=@n8, B=@n11)) {texture=/Game/Weapons/WeaponSkins/Textures/T_Swirly_Noise_D1}.G
   B=ScalarParameter() {parameterName=Text Distort Multiplier}
  )
  n13 = Add(A=Add(A=ComponentMask(None=@n6) {g=true, r=true}, B=Frac(None=@n7)), B=@n12)
  n14 = Multiply(A=WorldPosition().XYZ, B=ScalarParameter() {defaultValue=0.0003, parameterName=World UV2 Scale})
  n15 = Multiply(A=Time(), B=AppendVector(A=ScalarParameter() {parameterName=Tex 2 Speed U}, B=ScalarParameter() {parameterName=Tex 2 Speed V}))
  n16 = Add(A=Add(A=ComponentMask(None=@n14) {g=true, r=true}, B=Frac(None=@n15)), B=@n12)
  n17 = Add(
   A=Power(
    Base=TextureSample(UVs=@n13) {samplerType=SAMPLERTYPE_LinearGrayscale, texture=/Figment_S10/Effects/Fort_Effects/BluGlow/Violentium/T_SwirlNoise_N_height}.G
    Exp=ScalarParameter() {defaultValue=3, parameterName=AlphaTexPower_01}
   )
   B=Power(
    Base=TextureSample(UVs=@n16) {samplerType=SAMPLERTYPE_LinearGrayscale, texture=&T_SwirlNoise_N_height}.G
    Exp=ScalarParameter() {defaultValue=3, parameterName=AlphaTexPower_02}
   )
  )
  n18 = WorldPosition()
  n19 = ScalarParameter() {defaultValue=0.0003, parameterName=World UV3 Scale}
  n20 = Time()
  n21 = AppendVector(
   A=ScalarParameter() {defaultValue=0.0001, parameterName=Tex 3 Speed U}
   B=ScalarParameter() {defaultValue=0.0001, parameterName=Tex 3 Speed V}
  )
  n22 = Add(A=ComponentMask(None=Multiply(A=@n18.XYZ, B=@n19)) {g=true, r=true}, B=Frac(None=Multiply(A=@n20, B=@n21)))
  n23 = Power(
   Base=TextureSample(UVs=Add(A=@n22, B=@n12)) {texture=&T_Swirly_Noise_D1}.G
   Exp=ScalarParameter() {defaultValue=3, parameterName=AlphaTexPower_03}
  )
  n24 = Multiply(
   A=Divide(A=Add(A=@n17, B=@n23)) {constB=3}
   B=MaterialFunctionCall(In=VertexColor().A, Contrast=ScalarParameter() {parameterName=Contrast}) {materialFunction=&CheapContrast}.Result
  )
  n25 = Clamp(None=MaterialFunctionCall(
    Fade Length=ScalarParameter() {defaultValue=1500, parameterName=Fade Close Length}
    Fade Offset=ScalarParameter() {defaultValue=150, parameterName=Fade Close Offset}
   ) {materialFunction=&CameraDepthFade}.Result)
  n26 = Multiply(A=WorldPosition().XYZ, B=ScalarParameter() {defaultValue=0.0003, parameterName=World UV3 Scale})
  n27 = Multiply(
   A=Time()
   B=AppendVector(A=ScalarParameter() {parameterName=Tex Depth Fade Speed U}, B=ScalarParameter() {parameterName=Tex Depth Fade Speed V})
  )
  n28 = Add(A=Add(A=ComponentMask(None=@n26) {g=true, r=true}, B=Frac(None=@n27)), B=Multiply(A=@n12) {constB=3})
  DepthFade(
   Opacity=Multiply(A=Power(Base=@n4, Exp=@n5), B=Multiply(A=@n24, B=@n25))
   FadeDistance=LinearInterpolate(
    A=ScalarParameter() {parameterName=Fade Distance 2}
    B=ScalarParameter() {parameterName=Fade Distance 1}
    Alpha=TextureSample(UVs=@n28) {samplerType=SAMPLERTYPE_LinearGrayscale, texture=&T_SwirlNoise_N_height}.G
   )
  )
  LinearInterpolate()
  Multiply(
   A=MaterialFunctionCall(
    Color_Morning=VectorParameter() {parameterName=Color Morning, defaultValue=<rgba 0.8021 0.8021 0.8021 1>}.RGB
    Color_Day=VectorParameter() {parameterName=Color Day, defaultValue=<rgba 0.8021 0.8021 0.8021 1>}.RGB
    Color_Evening=VectorParameter() {parameterName=Color Evening, defaultValue=<rgba 0.8021 0.8021 0.8021 1>}.RGB
    Color_Night=VectorParameter() {parameterName=Color Night, defaultValue=<rgba 0.8021 0.8021 0.8021 1>}.RGB
   ) {materialFunction=&Color_Time_of_Day}.Result
   B=ScalarParameter() {parameterName=Emissive Mult}
  ) {constB=0.0025}
  Power(Base=DotProduct(A=CameraVectorWS(), B=VertexNormalWS()), Exp=ScalarParameter() {parameterName=Angle Fade Power})
 ]
}
